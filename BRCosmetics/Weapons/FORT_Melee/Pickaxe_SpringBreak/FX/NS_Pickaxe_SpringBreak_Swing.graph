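# Niagara swing trail effect for the Spring Break pickaxe, drawn as two short-lived ribbons driven by skeletal mesh anim trail modules.
# BladeTrail spawns between two specified bone or socket indices and fades a neutral, brightened color over life.
# Light_Trail adds a pale blue ribbon that narrows over life, with alpha ramping in by emitter age and dimming at night through a time-of-day scalar.
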NiagaraSystem /BRCosmetics/Weapons/FORT_Melee/Pickaxe_SpringBreak/FX/NS_Pickaxe_SpringBreak_Swing
dependencies {
 usedDataInterfaces [&NiagaraDataInterfaceColorCurve, &NiagaraDataInterfaceCurve, &NiagaraDataInterfaceSkeletalMesh]
 usedDynamicInputs [
  &Multiply_Vector2
  &RandomRangeVector2D
  &Vector2DFromFloat
  &RandomRangeFloat
 ]
 usedModules [
  &self:SystemSpawnScript.SystemScriptSource.SystemScriptGraph.NiagaraNodeAssignment_0.SetVariables_#1
  &self:SystemSpawnScript.SystemScriptSource.SystemScriptGraph.NiagaraNodeAssignment_2.SetVariables_#2
  &SystemState
  &self:SystemSpawnScript.SystemScriptSource.SystemScriptGraph.NiagaraNodeAssignment_1.SetVariables_#3
 ]
}
emitters {
 BladeTrail {
  emitterUpdateScript [
   &EmitterState()
   &FN_SpawnAnimTrail()
  ]
  particleSpawnScript [
   &FN_InitializeAnimTrailBoneIndecies(First Bone or Socket Index=2, Second Bone or Socket Index=3)
   &FN_InitializeAnimTrail(Color=<rgba 0.5 0.5 0.5 0.5>, Lifetime=0.15, SkeletalMesh=$System.Skeletal Mesh)
   /SetVariables() {isSetParameters=true}
  ]
  particleUpdateScript [
   &SolveForcesAndVelocity(Clamp Velocity=false, Limit Acceleration=false)
   &FN_UpdateAnimTrail(Locked Particle RibbonWidth=1, SkeletalMesh=$User.Skeletal Mesh) {module=FN_UpdateAnimTrail001}
   /SetVariables(Particles.ColorCurveAlt=DataInterface() {propertyValues {AlphaCurve {keys []}, BlueCurve {keys [<at 0.008 1>]}, GreenCurve {keys [<at 0.008 1>]}, RedCurve {keys [<at 0.008 1>]}}}) {isSetParameters=true}
   &ScaleColor(
    Scale Alpha=&FloatFromCurve(
     FloatCurve=DataInterface() {propertyValues {Curve {keys [
         {arriveTangent=-2.003, interpMode=RCIM_Cubic, leaveTangent=-2.003, tangentMode=RCTM_User, time=0.5, value=0.95}
         {interpMode=RCIM_Cubic, tangentMode=RCTM_User, time=1, value=0}
        ]}}}
     Scale Curve=0.35
    )
    Scale RGB=<2 2 2>
    ScaleA=true
    ScaleRGB=true
   )
  ]
  rendererClasses [&NiagaraRibbonRendererProperties]
  renderers [{rendererClass=&NiagaraRibbonRendererProperties}]
 }
 Light_Trail {
  emitterUpdateScript [
   &EmitterState()
   &FN_SpawnAnimTrail()
   &DoOnce(
    Reset:NiagaraBool=&self:ScratchDynamicInput(ENiagara Execution State:ENiagaraExecutionState=$System.ExecutionState, ENiagara Execution State001:ENiagaraExecutionState=::Active)
    Trigger Condition:NiagaraBool=&self:ScratchDynamicInput(ENiagara Execution State:ENiagaraExecutionState=$System.ExecutionState, ENiagara Execution State001:ENiagaraExecutionState=::Inactive)
   )
   &InterpolateOverTime(
    Rate Of Change:NiagaraFloat=&Divide_Float(A=1, B=0.05)
    Target Float Value:NiagaraFloat=&MakeCustomFloatFromBool(Bool=$Emitter.DoOnce.TriggeredOnce, True Float=1.5, False Float=0)
   )
  ]
  particleSpawnScript [
   &FN_InitializeAnimTrailBoneIndecies()
   &FN_InitializeAnimTrail(
    Color=&MultiplyLinearColor(
     Linear Color=<rgba 0.65 0.7667 1 0.35>
     Scale Factor=&MakeFloat4FromVectorAndFloat(
      "Vector (RGB)"=&VectorFromFloat(Value=5)
      "Float (Alpha)"=&Multiply_Float(
       A=1
       B=&FloatFromCurve(
        FloatCurve=DataInterface() {propertyValues {Curve {keys [{tangentMode=RCTM_User, time=0, value=0}, {tangentMode=RCTM_User, time=0.2, value=1}]}}}
        CurveIndex=$Emitter.Age
       )
      )
     )
    )
    Lifetime=0.11
    SkeletalMesh=$System.Skeletal Mesh
   )
  ]
  particleUpdateScript [
   &SolveForcesAndVelocity(Clamp Velocity=false, Limit Acceleration=false)
   &FN_UpdateAnimTrail(Locked Particle RibbonWidth=1, SkeletalMesh=$User.Skeletal Mesh) {module=FN_UpdateAnimTrail001}
   &ScaleColor(
    Scale Alpha=&FloatFromCurve(
     FloatCurve=DataInterface() {propertyValues {Curve {keys [<at 0.1 1 RCIM_Cubic>,<at 1 0 RCIM_Cubic>]}}}
     Scale Curve=&FNTimeOfDayScalar(Night Scale=0.5, Morning Scale=1, Evening Scale=1)
    )
    Scale RGB=<0.2 0.2 0.2>
    ScaleA=true
    ScaleRGB=true
   ) {module=ScaleColor001}
   &ScaleRibbonWidth(Ribbon Width Scale=&FloatFromCurve(FloatCurve=DataInterface() {propertyValues {Curve {keys [{arriveTangent=-1.8706, interpMode=RCIM_Cubic, time=0, value=1}, {interpMode=RCIM_Cubic, tangentMode=RCTM_User, time=1, value=0.5}]}}}))
  ]
  rendererClasses [&NiagaraRibbonRendererProperties]
  renderers [{rendererClass=&NiagaraRibbonRendererProperties}]
 }
}
summary {
 systemName=NS_Pickaxe_SpringBreak_Swing
 emitters [
  {emitterName=BladeTrail, rendererClasses [&NiagaraRibbonRendererProperties]}
  {emitterName=Light_Trail, rendererClasses [&NiagaraRibbonRendererProperties]}
 ]
 userVariables [User.RibbonIDOffset {type=NiagaraInt32}]
}
