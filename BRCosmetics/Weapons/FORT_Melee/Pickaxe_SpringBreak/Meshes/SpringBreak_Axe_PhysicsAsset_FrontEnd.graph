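# Physics asset for the front-end variant of the Spring Break two-handed pickaxe, defining jointed capsule bodies for its dangling strap and balls.
# Four capsules shape the handle body, with base_hilt joined to it by a constraint with free swing and twist.
# From base_hilt, a tightly limited dyn_hilt leads into a four-segment strap chain, while a two-segment balls chain branches separately.
# Strap joints swing widely with almost no twist; ball joints have asymmetric swing limits.
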PhysicsAsset /BRCosmetics/Weapons/FORT_Melee/Pickaxe_SpringBreak/Meshes/SpringBreak_Axe_PhysicsAsset_FrontEnd
bodies {
 Melee_TwoHanded_Handle [
  {shapeType=Capsule, length=4.5, radius=7.5, center=<0 -1.6781 0.276>, rotation=<rot -0.0001 0 70>}
  {
   shapeType=Capsule
   length=10.5
   radius=1
   shapeName=Melee_TwoHanded_Handle_capsule
   center=<-1.0589 3.9849 -10.3081>
   rotation=<rot -0.0001 0 -90>
  }
  {
   shapeType=Capsule
   length=10.5
   radius=1
   shapeName=Melee_TwoHanded_Handle_capsule
   center=<-1.0589 2.5917 -8.3524>
   rotation=<rot -0.0001 0 -90>
  }
  {
   shapeType=Capsule
   length=10.5
   radius=1
   shapeName=Melee_TwoHanded_Handle_capsule
   center=<-1.0589 5.313 -12.3524>
   rotation=<rot -0.0001 0 -90>
  }
 ]
 base_hilt [{shapeType=Capsule, length=0.505, radius=0.505, shapeName=base_hilt_capsule, rotation=<rot 0 -90 0>}]
 dyn_balls [{shapeType=Capsule, length=1, radius=1, shapeName=dyn_balls_capsule, center=<2.4292 0 1.2787>, rotation=<rot -90 -108.7534 108.7534>}]
 dyn_balls_b [{shapeType=Capsule, length=2, radius=1, shapeName=dyn_balls_b_capsule, center=<4.0777 0 0.8246>, rotation=<rot -90 180 180>}]
 dyn_hilt [{shapeType=Capsule, length=1, radius=1, shapeName=dyn_hilt_capsule, center=<2 0 0>, rotation=<rot -90 0 0>}]
 dyn_strap_1 [{shapeType=Capsule, length=4.5, radius=1, shapeName=dyn_strap_1_capsule, center=<3 0 -1.0709>, rotation=<rot -90 -165.9638 -104.0362>}]
 dyn_strap_1_b [{shapeType=Capsule, length=4.5, radius=1, shapeName=dyn_strap_1_b_capsule, center=<3 0 0>, rotation=<rot -90 0 90>}]
 dyn_strap_2 [{shapeType=Capsule, length=2, radius=1, shapeName=dyn_strap_2_capsule, center=<3.2 0 -0.5474>, rotation=<rot -90 180 180>}]
 dyn_strap_3 [{shapeType=Capsule, length=2, radius=1, shapeName=dyn_strap_3_capsule, center=<3 0 0>, rotation=<rot -90 0 0>}]
}
constraints [
 {bone1Name=dyn_strap_3, bone2Name=dyn_strap_2, swing1LimitDegrees=70, swing2LimitDegrees=70, twistLimitDegrees=1}
 {
  bone1Name=base_hilt
  bone2Name=Melee_TwoHanded_Handle
  swing1LimitDegrees=45
  swing1Motion=Free
  swing2LimitDegrees=45
  swing2Motion=Free
  twistLimitDegrees=45
  twistMotion=Free
 }
 {bone1Name=dyn_hilt, bone2Name=base_hilt, swing1LimitDegrees=5, swing2LimitDegrees=5, twistLimitDegrees=1}
 {bone1Name=dyn_strap_1, bone2Name=dyn_hilt, swing1LimitDegrees=70, swing2LimitDegrees=70, twistLimitDegrees=1}
 {bone1Name=dyn_balls, bone2Name=base_hilt, swing1LimitDegrees=1, swing2LimitDegrees=45, twistLimitDegrees=25}
 {bone1Name=dyn_balls_b, bone2Name=dyn_balls, swing1LimitDegrees=60, swing2LimitDegrees=25, twistLimitDegrees=1}
 {bone1Name=dyn_strap_1_b, bone2Name=dyn_strap_1, swing1LimitDegrees=65, swing2LimitDegrees=65, twistLimitDegrees=0}
 {bone1Name=dyn_strap_2, bone2Name=dyn_strap_1_b, swing1LimitDegrees=60, swing2LimitDegrees=60, twistLimitDegrees=1}
]
